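# BrawnySmasher_Transition_Leadin_VO_Cue is a sound cue for the Brawny Smasher NPC's transition lead-in voice bark.
# A root Random node picks one of three equally weighted MotherWall lead-in wave players and is excluded from branch culling.
# The cue uses the NPC VO sound class, a Brawny Smasher attenuation asset and a boosted volume multiplier.
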SoundCue /Game/Sounds/Fort_Human_Barks/BrawnySmasher/BrawnySmasher_Transition_Leadin_VO_Cue
root=Random_0
nodes [
 {id=Random_0, children [WavePlayer_3,WavePlayer_4,WavePlayer_5], properties {bShouldExcludeFromBranchCulling=true, weights [1,1,1]}}
 {id=WavePlayer_3, properties {soundWaveAssetPtr=&/MotherWall_2021_LeadinToTransition_010}}
 {id=WavePlayer_4, properties {soundWaveAssetPtr=&/MotherWall_2021_LeadinToTransition_011}}
 {id=WavePlayer_5, properties {soundWaveAssetPtr=&/MotherWall_2021_LeadinToTransition_012}}
]
properties {
 attenuationSettings=/Game/Athena/Sounds/NPCs/BrawnySmasher/BrawnySmasher_Att
 maxDistance=4000
 soundClassObject=&Fort_VO_NPC
 volumeMultiplier=1.7
 concurrencySet [null]
}
